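# Translucent UI material that procedurally draws a View More button arrow with a scalable ghost arrow around it.
# Each arrow is a rounded-box SDF on mirrored UVs rotated 45 degrees, forming a chevron without textures.
# The ghost copy scales UVs around the arrow center and derives an anti-aliased outline from the SDF through ping-pong and threshold functions.
# Max nodes combine the arrow, ghost fill and ghost outline, each with its own opacity parameter.
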Material /FortUILibrary/UIKit/Buttons/Materials/Arrows/M_UIKit_ViewMore_Arrow
graph {
 properties {
  bUsedWithStaticMesh=false
  blendMode=BLEND_Translucent
  materialDomain=MD_UI
  pixelDepthOffsetMode=PDOM_Legacy
  usedShadingModels=Default Lit
 }
 nodes [
  n1 = TextureCoordinate()
  n2 = Constant2Vector() {g=-0.1, r=-0.15}
  n3 = MaterialFunctionCall(
   UVs=Multiply(
    A=MaterialFunctionCall!(UV=@n1, Offset=@n2, Rotation, Scale, MirrorY) {materialFunction=/FortUILibrary/MaterialFunctions/Basics/MF_UI_Symmetry_2x1_DEPRECATED}.Result
    B=Constant2Vector() {g=1, r=0.5}
   )
   Rotation Center=Constant2Vector() {g=0.5, r=0.5}
   "Rotation Angle (0-1)"=Constant() {r=0.125}
  ) {materialFunction=&CustomRotator}
  n4 = Constant2Vector() {g=0.5, r=0.125}
  n5 = Constant() {r=0.1}
  n6 = Constant() {r=0.02}
  n7 = ScalarParameter() {defaultValue=0.2, parameterName=Ghost Arrow Shape Opacity}
  n8 = MaterialFunctionCall(
   UV=MaterialFunctionCall!(
    Texture Scale=ScalarParameter() {defaultValue=2, parameterName=Ghost Arrow Scale}
    UVs=TextureCoordinate()
    Offset XY Texturesize Z=Constant3Vector() {desc=arrow center of scale, constant=<rgba 0.5 0.63 1 0>}
   ) {materialFunction=/Engine/Functions/Engine_MaterialFunctions01/Texturing/ScaleUVsAroundPoint}.UVs
   Offset=Constant2Vector() {g=-0.1, r=-0.15}
  ) {materialFunction=&MF_UI_Symmetry_2x1_DEPRECATED}
  n9 = Constant2Vector() {g=1, r=0.5}
  n10 = MaterialFunctionCall(
   UVs=MaterialFunctionCall(UVs=Multiply(A=@n8.Result, B=@n9), Rotation Center=Constant2Vector() {g=0.5, r=0.5}, "Rotation Angle (0-1)"=Constant() {r=0.125}) {materialFunction=&CustomRotator}.Rotated Values
   Size=Constant2Vector() {g=0.5, r=0.125}
   CornerRoundness=Constant() {r=0.05}
  ) {materialFunction=&MF_UI_Box}
  n11 = Add(A=Multiply(A=@n10.SDF, B=Constant() {r=6})) {constB=0.5}
  n12 = MaterialFunctionCall!(
   Threshold
   Gradient=MaterialFunctionCall!(Value=Saturate(None=@n11), Scale, Height) {materialFunction=/FortUILibrary/MaterialFunctions/Basics/MF_UI_PingPong}.Result
   Filter Width Override
   Filter Width Fudge Factor
  ) {materialFunction=/Game/UI/Materials/AAThreshold}
  n13 = ScalarParameter() {defaultValue=0.5, parameterName=Ghost Arrow Outline Opacity}
  Max(
   A=Multiply(
    A=MaterialFunctionCall(UVs=@n3.Rotated Values, Size=@n4, CornerRoundness=@n5, GlowMax=@n6) {materialFunction=&MF_UI_Box}.Fill
    B=ScalarParameter() {defaultValue=0.3, parameterName=Arrow Opacity}
   )
   B=Max(A=Multiply(A=@n7, B=@n10.Fill) {constB=0.5}, B=Multiply(A=@n12.Result, B=@n13))
  )
 ]
}
